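# Translucent UI material that draws a radial progress ring for consumables in the Battle Royale HUD.
# An anti-aliased SDF circle outline, its thickness driven by Anim_Thickness, forms the ring.
# Radial coordinates thresholded by Percent build progress, marker tip, marker shadow, eased glow and gradient masks, shared through named reroutes.
# These masks blend adaptive colors and collection white into color and opacity over a soft ring shadow, with low-end value switches.
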Material /BattleRoyaleGameUI/HUD/EquipProgressIndicator/M_UI_RadialProgress_Consumable
graph {
 properties {bAutomaticallySetUsageInEditor=false, blendMode=BLEND_Translucent, materialDomain=MD_UI, usedShadingModels=Default Lit}
 nodes [
  n1 = NamedRerouteUsage()
  n2 = Constant() {r=0.125}
  n3 = MaterialFunctionCall!(Default=Constant() {r=1}, LowEnd=Constant() {r=0.9}) {materialFunction=/BattleRoyaleGameUI/Shared/MF_UI_LowEnd_ValueSwitcher}
  n4 = MaterialFunctionCall(Default=Constant() {r=0.75}, LowEnd=Constant() {r=0.464}) {materialFunction=&MF_UI_LowEnd_ValueSwitcher}
  n5 = NamedRerouteUsage()
  n6 = OneMinus(None=@n5)
  n7 = Constant() {r=0.0825}
  n8 = MaterialFunctionCall(
   Size=Constant() {r=0.725}
   OutlineThickness=LinearInterpolate(A=@n7, B=Subtract(A=@n7) {constB=0.05}, Alpha=ScalarParameter() {parameterName=Anim_Thickness})
  ) {materialFunction=&MF_UI_Circle}
  n9 = MaterialFunctionCall(DistanceField=@n8.SDFOutline) {materialFunction=&MF_UI_AntiAliasing}
  LinearInterpolate(
   A=Multiply(A=Multiply(A=@n1, B=@n2), B=NamedRerouteUsage())
   Alpha=Multiply(A=LinearInterpolate(A=@n3.Result, B=@n4.Result, Alpha=@n6) {constA=1, constB=0.75}, B=Saturate(None=@n9.Inner))
  )
  n10 = MaterialFunctionCall!(BufferSource) {materialFunction=/BattleRoyaleGameUI/Shared/MF_UI_AdaptiveColors}
  n11 = ComponentMask(None=CollectionParameter() {collection=/BattleRoyaleGameUI/Shared/MPC_BRGameUI, parameterName=White}) {b=true, g=true, r=true}
  n12 = LinearInterpolate(A=@n10.Bright, B=@n11, Alpha=NamedRerouteUsage())
  LinearInterpolate(
   A=LinearInterpolate(A=LinearInterpolate(A=@n12, B=@n10.Dark, Alpha=@n6), B=@n11, Alpha=NamedRerouteUsage())
   B=MaterialFunctionCall(Default=@n10.Dim, LowEnd=@n10.Dim) {materialFunction=&MF_UI_LowEnd_ValueSwitcher}.Result
   Alpha=Min(A=NamedRerouteUsage(), B=@n9.Outer)
  )
  n13 = ComponentMask(None=MaterialFunctionCall!(Vector or UVs) {materialFunction=/FortUILibrary/MaterialFunctions/Basics/MF_UI_CheaperVectorToRadial}.Radial Coordinates) {r=true}
  n14 = ScalarParameter() {defaultValue=0.4, parameterName=Percent}
  n15 = MaterialFunctionCall(DistanceField=@n13, Threshold=OneMinus(None=Add(A=@n14) {constB=0.025}), Smoothing=Constant() {r=0.05}) {materialFunction=&MF_UI_AntiAliasing}
  n16 = MaterialFunctionCall!(A=ScalarParameter() {defaultValue=0.4, parameterName=Percent}, B=Constant()) {materialFunction=/FortUILibrary/MaterialFunctions/Basics/MF_UI_BooleanOperations}
  NamedRerouteDeclaration(None=LinearInterpolate(A=Saturate(None=@n15.Inner), Alpha=@n16.IsLessOrEqual) {constB=0}) {name=MarkerShadowMask, nodeColor=<rgba 0 0 0 1>}
  n17 = OneMinus(None=ScalarParameter() {defaultValue=0.4, parameterName=Percent})
  n18 = Saturate(None=MaterialFunctionCall(DistanceField=@n13, Threshold=@n17) {materialFunction=&MF_UI_AntiAliasing}.Outer)
  n19 = MaterialFunctionCall(DistanceField=@n13, Threshold=Add(A=@n17, B=Constant() {r=0.0125})) {materialFunction=&MF_UI_AntiAliasing}
  NamedRerouteDeclaration(None=Min(A=@n18, B=Saturate(None=@n19.Inner))) {name=MarkerTipMask, nodeColor=<rgba 0 0 0 1>}
  NamedRerouteDeclaration(None=@n18) {name=ProgressMask, nodeColor=<rgba 0 0 0 1>}
  NamedRerouteDeclaration(None=Min(A=@n9.Inner, B=@n5)) {name=RingProgressMask, nodeColor=<rgba 0 0 0 1>}
  n20 = Constant()
  n21 = Constant() {r=0.3}
  NamedRerouteDeclaration(None=Saturate(None=MaterialFunctionCall(DistanceField=@n8.SDFOutline, Threshold=@n20, Smoothing=@n21) {materialFunction=&MF_UI_AntiAliasing}.Inner)) {name=ShadowMask, nodeColor=<rgba 0 0 0 1>}
  NamedRerouteDeclaration(None=OneMinus(None=SmoothStep(Value=@n13))) {name=GradientMask, nodeColor=<rgba 0 0 0 1>}
  n22 = Constant() {r=0.05}
  n23 = MaterialFunctionCall("Value (0-1)"=Saturate(None=MaterialFunctionCall(DistanceField=@n13, Threshold=@n17, Smoothing=@n22) {materialFunction=&MF_UI_AntiAliasing}.Outer)) {materialFunction=&MF_UI_EaseCurves}
  NamedRerouteDeclaration(None=LinearInterpolate(A=Saturate(None=@n23.EaseInCubic), Alpha=@n16.IsLessOrEqual) {constB=0}) {name=ProgressGlowMask, nodeColor=<rgba 0 0 0 1>}
 ]
}
